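# AudioKnobLarge is a UMG knob widget for audio controls, layering images, a RadialSlider, an editable value box and label, units and min/max texts.
# Slider moves (normalized) and committed text (actual units) feed SetNewValue, which converts via ControlValueMin/Max without clamping, updates the slider and optionally text, then broadcasts OnControlValueChanged in actual units.
# ControlValue stores the raw input, so slider moves leave it normalized.
# Label, Units and DisplayMin/Max variables are never applied.
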WidgetBlueprint /AudioWidgets/AudioKnobLarge/AudioKnobLarge
graphs [
 ConditionText {nodes [Outputs!(ReturnValue:Text="To Text (Float)"(
     Value=Inputs!() -> (Value:double=~)
     MinimumIntegralDigits=1
     MaximumIntegralDigits=Get MaxIntegralDigits() -> (MaxIntegralDigits:int=0)
     MinimumFractionalDigits=0
     MaximumFractionalDigits=Get MaxFractionalDigits() -> (MaxFractionalDigits:int=0)
    ))]}
 EventGraph {nodes [
   n1 = "OnTextCommitted (Editable Text) (ValueTextBox)"!() -> (OutputDelegate, then, Text:"Text (by ref)", CommitMethod:ETextCommit Enum=Default) {is=title K2Node_ComponentBoundEvent}
   > SetNewValue!(execute, self, NewValue:double=String To Float(InString="To String (Text)"(InText=@n1.Text)), Normalized:bool=false, UpdateText:bool=false) -> (then)
   n2 = "On Value Changed (Slider)"!() -> (OutputDelegate, then, Value:float) {is=title K2Node_ComponentBoundEvent}
   > SetNewValue(NewValue=@n2.Value, Normalized=true, UpdateText=true)
   Event PreConstruct()
   Event Construct()
   Event Tick()
  ]}
 OnControlValueChanged {nodes [OnControlValueChanged!() -> (then, NewValue:double=~) {is=title K2Node_FunctionEntry}]}
 SetNewValue {nodes [
   n1 = SetNewValue!() -> (then, NewValue:double=~, Normalized:bool=true, UpdateText:bool=~) {is=title K2Node_FunctionEntry}
   > n2 = Branch(Condition=Get Normalized() -> (Normalized:bool))
   n3 = Get NewValue() -> (NewValue:double)
   > Set ControlValueNormalized(ControlValueNormalized:double=@n3) -> (Output_Get:double)
   > n4 = Branch(Condition=Get UpdateText() -> (UpdateText:bool))
   n5 = Lerp(A=Get ControlValueMin() -> (ControlValueMin:double), B=Get ControlValueMax() -> (ControlValueMax:double), Alpha=@n3)
   > n6 = "SetText (Editable Text)"!(
    execute
    self:Editable Text Object Reference=Get ValueTextBox() -> (ValueTextBox:Editable Text Object Reference)
    InText:Text=Condition Text!(Value:double=@n5) -> (ReturnValue:Text) {is=title K2Node_MacroInstance}
   ) -> (then)
   Set Value!(
    execute=[@n6.then, @n4.else]
    Value:float=Get ControlValueNormalized() -> (ControlValueNormalized:double)
    self:Radial Slider Object Reference=Get Slider() -> (Slider:Radial Slider Object Reference)
   ) -> (then, Output_Get:float)
   > n7 = Call On Control Value Changed!(execute:Exec, self:Audio Knob Large Object Reference, NewValue:double=@n5) -> (then:Exec) {is=title K2Node_CallDelegate}
   Set ControlValue(execute=[@n7.then, @n10.then], ControlValue:double=Get NewValue()) -> (Output_Get:double)
   K2Node_MathExpression_0 = "(NewValue - ControlValueMin) / (ControlValueMax - ControlValueMin)"!(NewValue:double=Get NewValue()) -> (ReturnValue:double=~) {
    is=title K2Node_MathExpression
    expression=((NewValue - {#1}) / ({#2} - {#1}))
   }
   Set ControlValueNormalized(execute=@n2.else, ControlValueNormalized=@K2Node_MathExpression_0)
   > n8 = Branch(Condition=Get UpdateText())
   > n9 = "SetText (Editable Text)"(self=Get ValueTextBox(), InText=Condition Text(Value=@n1))
   Set Value(execute=[@n9.then, @n8.else], Value=Get ControlValueNormalized(), self=Get Slider())
   > n10 = Call On Control Value Changed(NewValue=Get NewValue())
  ]}
 MathExpression_2 {
  path=SetNewValue.K2Node_MathExpression_0.MathExpression_2
  nodes [
   n1 = Inputs!() -> (NewValue:double=~)
   n2 = Get ControlValueMin()
   n3 = Get ControlValueMax()
   n4 = Get ControlValueMin()
   Outputs(ReturnValue:double=float / float(A=float - float(A=@n1, B=@n2) {is=title K2Node_CallFunction}, B=float - float(A=@n3, B=@n4)) {is=title K2Node_CallFunction})
  ]
 }
]
variables [
 ControlValue {type=number}
 ControlValueMax {type=number, default=1}
 ControlValueMin {type=number}
 ControlValueNormalized {type=number}
 DisplayMax {type=string, default=1}
 DisplayMin {type=string, default=0}
 Label {type=string, default=Control01}
 MaxFractionalDigits {type=integer, default=3}
 MaxIntegralDigits {type=integer, default=3}
 OnControlValueChanged {}
 ToolTip {type=string, default=""}
 Units {type=string, default=Units}
]
widgets {
 info {inheritedWidgetCount=0, parentClass=&UserWidget, rootWidgetClass=&CanvasPanel, widgetCount=11}
 widgets [
  Panel01 {widgetClassPath=&CanvasPanel, widget=Panel01}
  Layer00 {
   widgetClassPath=&Image
   bIsVariable=true
   parent=Panel01
   widget=Layer00
   properties {brush {
     bIsDynamicallyLoaded=false
     drawAs=Image
     imageType=NoImage
     mirroring=NoMirror
     resourceName=None
     resourceObject=&/AudioKnobLarge_Layer00
     tiling=NoTile
     imageSize=<32 32>
     margin=<ltrb 0 0 0 0>
     outlineSettings {
      bUseBrushTransparency=false
      roundingType=HalfHeightRadius
      width=0
      color {colorUseRule=UseColor_Specified, specifiedColor=<rgba 0 0 0 0>}
      cornerRadii=<0 0 0 0>
     }
     tintColor {colorUseRule=UseColor_Specified, specifiedColor=<rgba 1 1 1 1>}
     uVRegion {bIsValid=false, max=<0 0>, min=<0 0>}
    }}
   slot {class=CanvasPanelSlot, properties {layoutData {alignment=<0 0>, anchors {maximum=<1 1>, minimum=<0 0>}, offsets=<ltrb 0 0 0 0>}}}
  }
  Layer01 {
   widgetClassPath=&Image
   bIsVariable=true
   parent=Panel01
   widget=Layer01
   properties {brush {
     bIsDynamicallyLoaded=false
     drawAs=Image
     imageType=NoImage
     mirroring=NoMirror
     resourceName=None
     resourceObject=&/AudioKnobLarge_Layer01
     tiling=NoTile
     imageSize=<32 32>
     margin=<ltrb 0 0 0 0>
     outlineSettings {
      bUseBrushTransparency=false
      roundingType=HalfHeightRadius
      width=0
      color {colorUseRule=UseColor_Specified, specifiedColor=<rgba 0 0 0 0>}
      cornerRadii=<0 0 0 0>
     }
     tintColor {colorUseRule=UseColor_Specified, specifiedColor=<rgba 1 1 1 1>}
     uVRegion {bIsValid=false, max=<0 0>, min=<0 0>}
    }}
   slot {class=CanvasPanelSlot, properties {layoutData {alignment=<0.5 0>, anchors {maximum=<0.5 0>, minimum=<0.5 0>}, offsets=<ltrb 0 12 110 94>}}}
  }
  Layer02 {
   widgetClassPath=&Image
   bIsVariable=true
   parent=Panel01
   widget=Layer02
   properties {brush {
     bIsDynamicallyLoaded=false
     drawAs=Image
     imageType=NoImage
     mirroring=NoMirror
     resourceName=None
     resourceObject=&/AudioKnobLarge_Layer02
     tiling=NoTile
     imageSize=<32 32>
     margin=<ltrb 0 0 0 0>
     outlineSettings {
      bUseBrushTransparency=false
      roundingType=HalfHeightRadius
      width=0
      color {colorUseRule=UseColor_Specified, specifiedColor=<rgba 0 0 0 0>}
      cornerRadii=<0 0 0 0>
     }
     tintColor {colorUseRule=UseColor_Specified, specifiedColor=<rgba 1 1 1 1>}
     uVRegion {bIsValid=false, max=<0 0>, min=<0 0>}
    }}
   slot {class=CanvasPanelSlot, properties {layoutData {alignment=<0.5 0>, anchors {maximum=<0.5 0>, minimum=<0.5 0>}, offsets=<ltrb 0 22 90 90>}}}
  }
  Slider {
   widgetClassPath=/Script/AdvancedWidgets.RadialSlider
   bIsVariable=true
   parent=Panel01
   widget=Slider
   properties {
    angularOffset=359.5
    sliderHandleEndAngle=312
    sliderHandleStartAngle=49
    sliderBarColor=<rgba 0.5 0.5 0.5 0>
    sliderHandleColor=<rgba 1 1 1 0>
    sliderProgressColor=<rgba 0.1022 0.0343 0.3185 1>
   }
   slot {class=CanvasPanelSlot, properties {layoutData {alignment=<0.5 0>, anchors {maximum=<0.5 0>, minimum=<0.5 0>}, offsets=<ltrb 0 14 105 105>}}}
  }
  EditableTextHorizontalBox {
   widgetClassPath=&HorizontalBox
   parent=Panel01
   widget=EditableTextHorizontalBox
   slot {
    class=CanvasPanelSlot
    properties {bAutoSize=true, zOrder=3, layoutData {alignment=<0.5 0.5>, anchors {maximum=<0.5 0>, minimum=<0.5 0>}, offsets=<ltrb 0.6511 64 166 32>}}
   }
  }
  ValueTextBox {
   widgetClassPath=/Script/UMG.EditableText
   bIsVariable=true
   parent=EditableTextHorizontalBox
   widget=ValueTextBox
   properties {
    justification=Right
    text=0
    widgetStyle {
     backgroundImageComposing {
      bIsDynamicallyLoaded=false
      drawAs=Border
      imageType=FullColor
      mirroring=NoMirror
      resourceName=../../../Engine/Content/Slate/Old/HyperlinkDotted.png
      resourceObject=null
      tiling=Both
      imageSize=<0 0>
      margin=<ltrb 0 0 0 0.1875>
      outlineSettings {
       bUseBrushTransparency=false
       roundingType=HalfHeightRadius
       width=0
       color {colorUseRule=UseColor_Specified, specifiedColor=<rgba 0 0 0 0>}
       cornerRadii=<0 0 0 0>
      }
      tintColor {colorUseRule=UseColor_Specified, specifiedColor=<rgba 1 1 1 1>}
      uVRegion {bIsValid=false, max=<0 0>, min=<0 0>}
     }
     backgroundImageSelected {
      bIsDynamicallyLoaded=false
      drawAs=Box
      imageType=FullColor
      mirroring=NoMirror
      resourceName=../../../Engine/Content/Slate/Common/EditableTextSelectionBackground.png
      resourceObject=null
      tiling=NoTile
      imageSize=<0 0>
      margin=<ltrb 0.25 0.25 0.25 0.25>
      outlineSettings {
       bUseBrushTransparency=false
       roundingType=HalfHeightRadius
       width=0
       color {colorUseRule=UseColor_Specified, specifiedColor=<rgba 0 0 0 0>}
       cornerRadii=<0 0 0 0>
      }
      tintColor {colorUseRule=UseColor_Specified, specifiedColor=<rgba 1 1 1 1>}
      uVRegion {bIsValid=false, max=<0 0>, min=<0 0>}
     }
     caretImage {
      bIsDynamicallyLoaded=false
      drawAs=Image
      imageType=FullColor
      mirroring=NoMirror
      resourceName=../../../Engine/Content/Slate/Old/White.png
      resourceObject=null
      tiling=NoTile
      imageSize=<16 16>
      margin=<ltrb 0 0 0 0>
      outlineSettings {
       bUseBrushTransparency=false
       roundingType=HalfHeightRadius
       width=0
       color {colorUseRule=UseColor_Specified, specifiedColor=<rgba 0 0 0 0>}
       cornerRadii=<0 0 0 0>
      }
      tintColor {colorUseRule=UseColor_Specified, specifiedColor=<rgba 1 1 1 1>}
      uVRegion {bIsValid=false, max=<0 0>, min=<0 0>}
     }
     colorAndOpacity {colorUseRule=UseColor_Specified, specifiedColor=<rgba 1 1 1 1>}
     font {
      bForceMonospaced=false
      bMaterialIsStencil=false
      fontMaterial=null
      fontObject=&Roboto
      letterSpacing=0
      monospacedWidth=1
      size=11
      skewAmount=0
      typefaceFontName=Regular
      outlineSettings {
       bApplyOutlineToDropShadows=false
       bMiteredCorners=false
       bSeparateFillAlpha=false
       outlineMaterial=null
       outlineSize=0
       outlineColor=<rgba 0 0 0 1>
      }
     }
     hintColorAndOpacity {colorUseRule=UseColor_Foreground, specifiedColor=<rgba 1 0 1 1>}
     selectedBackgroundColor {colorUseRule=UseColor_Foreground, specifiedColor=<rgba 1 0 1 1>}
    }
   }
   slot {
    class=HorizontalBoxSlot
    properties {horizontalAlignment=HAlign_Center, verticalAlignment=VAlign_Center, size {sizeRule=Automatic, value=0.6}}
   }
  }
  UnitsText {
   widgetClassPath=&EditableText
   parent=EditableTextHorizontalBox
   widget=UnitsText
   properties {
    justification=Right
    text=Units
    visibility=HitTestInvisible
    widgetStyle {
     backgroundImageComposing {
      bIsDynamicallyLoaded=false
      drawAs=Border
      imageType=FullColor
      mirroring=NoMirror
      resourceName=../../../Engine/Content/Slate/Old/HyperlinkDotted.png
      resourceObject=null
      tiling=Both
      imageSize=<0 0>
      margin=<ltrb 0 0 0 0.1875>
      outlineSettings {
       bUseBrushTransparency=false
       roundingType=HalfHeightRadius
       width=0
       color {colorUseRule=UseColor_Specified, specifiedColor=<rgba 0 0 0 0>}
       cornerRadii=<0 0 0 0>
      }
      tintColor {colorUseRule=UseColor_Specified, specifiedColor=<rgba 1 1 1 1>}
      uVRegion {bIsValid=false, max=<0 0>, min=<0 0>}
     }
     backgroundImageSelected {
      bIsDynamicallyLoaded=false
      drawAs=Box
      imageType=FullColor
      mirroring=NoMirror
      resourceName=../../../Engine/Content/Slate/Common/EditableTextSelectionBackground.png
      resourceObject=null
      tiling=NoTile
      imageSize=<0 0>
      margin=<ltrb 0.25 0.25 0.25 0.25>
      outlineSettings {
       bUseBrushTransparency=false
       roundingType=HalfHeightRadius
       width=0
       color {colorUseRule=UseColor_Specified, specifiedColor=<rgba 0 0 0 0>}
       cornerRadii=<0 0 0 0>
      }
      tintColor {colorUseRule=UseColor_Specified, specifiedColor=<rgba 1 1 1 1>}
      uVRegion {bIsValid=false, max=<0 0>, min=<0 0>}
     }
     caretImage {
      bIsDynamicallyLoaded=false
      drawAs=Image
      imageType=FullColor
      mirroring=NoMirror
      resourceName=../../../Engine/Content/Slate/Old/White.png
      resourceObject=null
      tiling=NoTile
      imageSize=<16 16>
      margin=<ltrb 0 0 0 0>
      outlineSettings {
       bUseBrushTransparency=false
       roundingType=HalfHeightRadius
       width=0
       color {colorUseRule=UseColor_Specified, specifiedColor=<rgba 0 0 0 0>}
       cornerRadii=<0 0 0 0>
      }
      tintColor {colorUseRule=UseColor_Specified, specifiedColor=<rgba 1 1 1 1>}
      uVRegion {bIsValid=false, max=<0 0>, min=<0 0>}
     }
     colorAndOpacity {colorUseRule=UseColor_Specified, specifiedColor=<rgba 1 1 1 1>}
     font {
      bForceMonospaced=false
      bMaterialIsStencil=false
      fontMaterial=null
      fontObject=&Roboto
      letterSpacing=0
      monospacedWidth=1
      size=11
      skewAmount=0
      typefaceFontName=Regular
      outlineSettings {
       bApplyOutlineToDropShadows=false
       bMiteredCorners=false
       bSeparateFillAlpha=false
       outlineMaterial=null
       outlineSize=0
       outlineColor=<rgba 0 0 0 1>
      }
     }
     hintColorAndOpacity {colorUseRule=UseColor_Foreground, specifiedColor=<rgba 1 0 1 1>}
     selectedBackgroundColor {colorUseRule=UseColor_Foreground, specifiedColor=<rgba 1 0 1 1>}
    }
   }
   slot {
    class=HorizontalBoxSlot
    properties {horizontalAlignment=HAlign_Center, verticalAlignment=VAlign_Center, padding=<ltrb 3 0 0 0>, size {sizeRule=Automatic, value=0.4}}
   }
  }
  LabelTextBox {
   widgetClassPath=&TextBlock
   parent=Panel01
   widget=LabelTextBox
   properties {
    justification=Center
    text=Control01
    visibility=HitTestInvisible
    colorAndOpacity {colorUseRule=UseColor_Specified, specifiedColor=<rgba 0.3813 0.3813 0.3813 1>}
    font {
     bForceMonospaced=false
     bMaterialIsStencil=false
     fontMaterial=null
     fontObject=&Roboto
     letterSpacing=0
     monospacedWidth=1
     size=13
     skewAmount=0
     typefaceFontName=Regular
     outlineSettings {
      bApplyOutlineToDropShadows=false
      bMiteredCorners=false
      bSeparateFillAlpha=false
      outlineMaterial=null
      outlineSize=0
      outlineColor=<rgba 0 0 0 1>
     }
    }
    strikeBrush {
     bIsDynamicallyLoaded=false
     drawAs=Image
     imageType=NoImage
     mirroring=NoMirror
     resourceName=None
     resourceObject=null
     tiling=NoTile
     imageSize=<32 32>
     margin=<ltrb 0 0 0 0>
     outlineSettings {
      bUseBrushTransparency=false
      roundingType=HalfHeightRadius
      width=0
      color {colorUseRule=UseColor_Specified, specifiedColor=<rgba 0 0 0 0>}
      cornerRadii=<0 0 0 0>
     }
     tintColor {colorUseRule=UseColor_Specified, specifiedColor=<rgba 1 0.9447 0.8928 1>}
     uVRegion {bIsValid=false, max=<0 0>, min=<0 0>}
    }
   }
   slot {
    class=CanvasPanelSlot
    properties {bAutoSize=true, layoutData {alignment=<0.5 1>, anchors {maximum=<0.5 1>, minimum=<0.5 1>}, offsets=<ltrb 0 -12 100 30>}}
   }
  }
  DisplayMinTextBox {
   widgetClassPath=&TextBlock
   parent=Panel01
   widget=DisplayMinTextBox
   properties {
    justification=Center
    text=Min
    visibility=HitTestInvisible
    font {
     bForceMonospaced=false
     bMaterialIsStencil=false
     fontMaterial=null
     fontObject=&Roboto
     letterSpacing=0
     monospacedWidth=1
     size=11
     skewAmount=0
     typefaceFontName=Light
     outlineSettings {
      bApplyOutlineToDropShadows=false
      bMiteredCorners=false
      bSeparateFillAlpha=false
      outlineMaterial=null
      outlineSize=0
      outlineColor=<rgba 0 0 0 1>
     }
    }
    strikeBrush {
     bIsDynamicallyLoaded=false
     drawAs=Image
     imageType=NoImage
     mirroring=NoMirror
     resourceName=None
     resourceObject=null
     tiling=NoTile
     imageSize=<32 32>
     margin=<ltrb 0 0 0 0>
     outlineSettings {
      bUseBrushTransparency=false
      roundingType=HalfHeightRadius
      width=0
      color {colorUseRule=UseColor_Specified, specifiedColor=<rgba 0 0 0 0>}
      cornerRadii=<0 0 0 0>
     }
     tintColor {colorUseRule=UseColor_Specified, specifiedColor=<rgba 1 0.9447 0.8928 1>}
     uVRegion {bIsValid=false, max=<0 0>, min=<0 0>}
    }
   }
   slot {
    class=CanvasPanelSlot
    properties {bAutoSize=true, layoutData {alignment=<0 1>, anchors {maximum=<0.5 0>, minimum=<0.5 0>}, offsets=<ltrb -48 128 93 30>}}
   }
  }
  DisplayMaxTextBox {
   widgetClassPath=&TextBlock
   parent=Panel01
   widget=DisplayMaxTextBox
   properties {
    justification=Center
    text=Max
    visibility=HitTestInvisible
    font {
     bForceMonospaced=false
     bMaterialIsStencil=false
     fontMaterial=null
     fontObject=&Roboto
     letterSpacing=0
     monospacedWidth=1
     size=11
     skewAmount=0
     typefaceFontName=Light
     outlineSettings {
      bApplyOutlineToDropShadows=false
      bMiteredCorners=false
      bSeparateFillAlpha=false
      outlineMaterial=null
      outlineSize=0
      outlineColor=<rgba 0 0 0 1>
     }
    }
    strikeBrush {
     bIsDynamicallyLoaded=false
     drawAs=Image
     imageType=NoImage
     mirroring=NoMirror
     resourceName=None
     resourceObject=null
     tiling=NoTile
     imageSize=<32 32>
     margin=<ltrb 0 0 0 0>
     outlineSettings {
      bUseBrushTransparency=false
      roundingType=HalfHeightRadius
      width=0
      color {colorUseRule=UseColor_Specified, specifiedColor=<rgba 0 0 0 0>}
      cornerRadii=<0 0 0 0>
     }
     tintColor {colorUseRule=UseColor_Specified, specifiedColor=<rgba 1 0.9447 0.8928 1>}
     uVRegion {bIsValid=false, max=<0 0>, min=<0 0>}
    }
   }
   slot {
    class=CanvasPanelSlot
    properties {bAutoSize=true, layoutData {alignment=<1 1>, anchors {maximum=<0.5 0>, minimum=<0.5 0>}, offsets=<ltrb 52 128 93 30>}}
   }
  }
 ]
}
