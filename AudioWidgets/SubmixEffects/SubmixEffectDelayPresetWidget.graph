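# Editor widget for Submix Effect Delay presets, titled "Delay Editor", with large knobs for Delay, Max Delay and Interp Time.
# Knob changes update the preset's settings and apply them through SetDefaultSettings.
# Construction and property-change events cast and store the preset, then OnPresetUpdated refreshes all knobs or the one named property.
# The delay and max delay setters detect when their change alters the other value and resync that knob.
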WidgetBlueprint /AudioWidgets/SubmixEffects/SubmixEffectDelayPresetWidget
graphs [
 EventGraph {nodes [
   n1 = "On Control Value Changed (LargeKnobDelay)"!() -> (OutputDelegate, then, NewValue:double) {is=title K2Node_ComponentBoundEvent}
   > SetDelayLength!(execute, self, DelayLength:double=@n1.NewValue) -> (then)
   n2 = "On Control Value Changed (LargeKnobDelayMax)"!() -> (OutputDelegate, then, NewValue:double) {is=title K2Node_ComponentBoundEvent}
   > SetDelayMaxLength!(execute, self, DelayLengthMax:double=@n2.NewValue) -> (then)
   n3 = "On Control Value Changed (LargeKnobTimeInterp)"!() -> (OutputDelegate, then, NewValue:double) {is=title K2Node_ComponentBoundEvent}
   > SetInterpTime!(execute, self, InterpTime:double=@n3.NewValue) -> (then)
   n4 = Event On SoundEffectPreset Widget Constructed!() -> (OutputDelegate, then, Preset:Sound Effect Preset Object Reference) {is=title K2Node_Event}
   > n5 = Cast To SubmixEffectDelayPreset!(execute, Object=@n4.Preset) -> (then, CastFailed, AsSubmix Effect Delay Preset:Submix Effect Delay Preset Object Reference) {is=title K2Node_DynamicCast}
   > Set Preset(Preset:Submix Effect Delay Preset Object Reference=@n5) -> (Output_Get:Submix Effect Delay Preset Object Reference)
   > OnPresetUpdated!(execute, self, Selection:Name=None) -> (then)
   n6 = Event OnPropertyChanged!() -> (OutputDelegate, then, Preset:Sound Effect Preset Object Reference, PropertyName:Name=None) {is=title K2Node_Event}
   > n7 = Cast To SubmixEffectDelayPreset(Object=@n6.Preset)
   > Set Preset(Preset=@n7)
   > OnPresetUpdated(Selection=@n6.PropertyName)
  ]}
 GetClass {nodes [
   GetClass() {is=title K2Node_FunctionEntry}
   > Return Node!(execute, ReturnValue:Sound Effect Preset Class Reference=/Script/Synthesis.SubmixEffectDelayPreset)
  ]}
 GetEditorName {nodes [
   GetEditorName() {is=title K2Node_FunctionEntry}
   > Return Node(ReturnValue:Text="LOCTEXT(\"Delay Editor\")")
  ]}
 GetIconBrushName {nodes [
   GetIconBrushName() {is=title K2Node_FunctionEntry}
   > Return Node(ReturnValue:Name=None)
  ]}
 OnPresetUpdated {nodes [
   n1 = OnPresetUpdated!() -> (then, Selection:Name) {is=title K2Node_FunctionEntry}
   > n2 = Switch on Name!(execute:Exec, Selection:Name=@n1) -> (Default:Exec, MaximumDelayLength:Exec, InterpolationTime:Exec, DelayLength:Exec) {is=title K2Node_SwitchName, pinNames [MaximumDelayLength, InterpolationTime, DelayLength]}
   n3 = Break Submix Effect Delay Settings!(SubmixEffectDelaySettings:"Submix Effect Delay Settings Structure (by ref)"=Get Settings!(self:Submix Effect Delay Preset Object Reference=Get Preset() -> (Preset:Submix Effect Delay Preset Object Reference)) -> (Settings:Submix Effect Delay Settings Structure)) -> (MaximumDelayLength:float=~, InterpolationTime:float=~, DelayLength:float=~) {is=title K2Node_BreakStruct}
   > SetNewValue!(
    execute
    self:Audio Knob Large Object Reference=Get LargeKnobDelayMax() -> (LargeKnobDelayMax:Audio Knob Large Object Reference)
    NewValue:double=@n3
    Normalized:bool=false
    UpdateText:bool=true
   ) -> (then)
   > SetNewValue(
    self=Get LargeKnobTimeInterp() -> (LargeKnobTimeInterp:Audio Knob Large Object Reference)
    NewValue=@n3.InterpolationTime
    Normalized=false
    UpdateText=true
   )
   > SetNewValue(self=Get LargeKnobDelay() -> (LargeKnobDelay:Audio Knob Large Object Reference), NewValue=@n3.DelayLength, Normalized=false, UpdateText=true)
   n4 = Break Submix Effect Delay Settings(SubmixEffectDelaySettings=Get Settings(self=Get Preset())) -> (MaximumDelayLength=~, InterpolationTime=~, DelayLength=~)
   SetNewValue(execute=@n2.DelayLength, self=Get LargeKnobDelay(), NewValue=@n4.DelayLength, Normalized=false, UpdateText=true)
   SetNewValue(execute=@n2.InterpolationTime, self=Get LargeKnobTimeInterp(), NewValue=@n4.InterpolationTime, Normalized=false, UpdateText=true)
   SetNewValue(execute=@n2.MaximumDelayLength, self=Get LargeKnobDelayMax(), NewValue=@n4, Normalized=false, UpdateText=false)
  ]}
 SetDelayLength {nodes [
   SetDelayLength!() -> (then, DelayLength:double=~) {is=title K2Node_FunctionEntry}
   > Is Valid(InputObject=Get Preset())
   n1 = Get Preset()
   > Set LastMaxDelay(LastMaxDelay:double=Break Submix Effect Delay Settings(SubmixEffectDelaySettings=Get Settings(self=@n1)) -> (MaximumDelayLength=~, InterpolationTime=~, DelayLength=~)) -> (Output_Get:double)
   n2 = Get DelayLength() -> (DelayLength:double)
   > n3 = SetDelayLength!(execute, DelaySettings:"Submix Effect Delay Settings Structure (by ref)"=Get Settings(self=Get Preset()), DelayLength:float=@n2) -> (then, ReturnValue:Submix Effect Delay Settings Structure) {is=title K2Node_CallFunction}
   > SetDefaultSettings!(execute, self:Submix Effect Delay Preset Object Reference=Get Preset(), InSettings:"Submix Effect Delay Settings Structure (by ref)"=@n3) -> (then)
   n4 = Break Submix Effect Delay Settings(SubmixEffectDelaySettings=@n3) -> (MaximumDelayLength=~, InterpolationTime=~, DelayLength=~)
   n5 = Get LastMaxDelay() -> (LastMaxDelay:double)
   > Branch(Condition=NOT Boolean(A="Nearly Equal (Float)"(A=@n5, B=@n4, ErrorTolerance=0)))
   > OnPresetUpdated(Selection=MaximumDelayLength) {is=title K2Node_CallFunction}
   "Clamp (Float)"(Value=@n2, Min=0, Max=1)
  ]}
 SetDelayMaxLength {nodes [
   SetDelayMaxLength!() -> (then, DelayLengthMax:double=~) {is=title K2Node_FunctionEntry}
   > Is Valid(InputObject=Get Preset())
   n1 = Get Preset()
   > Set LastDelay(LastDelay:double=Break Submix Effect Delay Settings(SubmixEffectDelaySettings=Get Settings(self=@n1)) -> (MaximumDelayLength=~, InterpolationTime=~, DelayLength=~).DelayLength) -> (Output_Get:double)
   > n2 = SetMaximumDelayLength!(
    execute
    DelaySettings:"Submix Effect Delay Settings Structure (by ref)"=Get Settings(self=Get Preset())
    MaximumDelayLength:float=Get DelayLengthMax() -> (DelayLengthMax:double)
   ) -> (then, ReturnValue:Submix Effect Delay Settings Structure)
   > SetDefaultSettings(self=Get Preset(), InSettings=@n2)
   n3 = Break Submix Effect Delay Settings(SubmixEffectDelaySettings=@n2) -> (MaximumDelayLength=~, InterpolationTime=~, DelayLength=~)
   n4 = Get LastDelay() -> (LastDelay:double)
   > Branch(Condition=NOT Boolean(A="Nearly Equal (Float)"(A=@n4, B=@n3.DelayLength, ErrorTolerance=0)))
   > OnPresetUpdated(Selection=DelayLength)
  ]}
 SetInterpTime {nodes [
   SetInterpTime!() -> (then, InterpTime:double=~) {is=title K2Node_FunctionEntry}
   > Is Valid(InputObject=Get Preset())
   > n1 = SetInterpolationTime!(
    execute
    DelaySettings:"Submix Effect Delay Settings Structure (by ref)"=Get Settings(self=Get Preset())
    InterpolationTime:float=Get InterpTime() -> (InterpTime:double)
   ) -> (then, ReturnValue:Submix Effect Delay Settings Structure)
   > SetDefaultSettings(self=Get Preset(), InSettings=@n1)
  ]}
]
variables [Preset {type=SubmixEffectDelayPreset, default=null}]
widgets {
 info {inheritedWidgetCount=0, parentClass=&UserWidget, rootWidgetClass=&CanvasPanel, widgetCount=5}
 widgets [
  CanvasPanel_0 {widgetClassPath=&CanvasPanel, widget=CanvasPanel_0}
  KnobBox {
   widgetClassPath=&HorizontalBox
   parent=CanvasPanel_0
   widget=KnobBox
   slot {class=CanvasPanelSlot, properties {layoutData {alignment=<0 0>, anchors {maximum=<0 0>, minimum=<0 0>}, offsets=<ltrb 0 0 428 202>}}}
  }
  LargeKnobDelay {
   widgetClassPath=/AudioWidgets/AudioKnobLarge/AudioKnobLarge
   bIsVariable=true
   parent=KnobBox
   widget=LargeKnobDelay
   properties {controlValueMax=20000, displayMax=20 s, label=Delay, maxFractionalDigits=0, maxIntegralDigits=5, toolTip=Amount of delay, units=ms}
   slot {class=HorizontalBoxSlot, properties {padding=<ltrb 2 2 2 2>, size {sizeRule=Fill, value=1}}}
  }
  LargeKnobDelayMax {
   widgetClassPath=&AudioKnobLarge
   bIsVariable=true
   parent=KnobBox
   widget=LargeKnobDelayMax
   properties {
    controlValueMax=20000
    displayMax=20 s
    label=Max Delay
    maxFractionalDigits=0
    maxIntegralDigits=5
    toolTip=Maximum Delay supported by taps
    units=ms
   }
   slot {class=HorizontalBoxSlot, properties {padding=<ltrb 2 2 2 2>, size {sizeRule=Fill, value=1}}}
  }
  LargeKnobTimeInterp {
   widgetClassPath=&AudioKnobLarge
   bIsVariable=true
   parent=KnobBox
   widget=LargeKnobTimeInterp
   properties {controlValueMax=20000, displayMax=20 s, label=Interp Time, maxFractionalDigits=0, maxIntegralDigits=5, units=ms}
   slot {class=HorizontalBoxSlot, properties {padding=<ltrb 2 2 2 2>, size {sizeRule=Fill, value=1}}}
  }
 ]
}
